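# M_SparkHits_01 is an unlit translucent material for spark hit effects on Cascade and Niagara sprite particles.
# It multiplies the green channel of a 2x2 spark texture by particle color, then passes that through the shared emissive adjustment function.
# The same channel is also posterized by a floor-and-divide using step count parameters, with particle alpha shifting the quantization.
# A time-driven sine node feeds nothing visible.
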Material /Game/Effects/Fort_Effects/Materials/Sparks/M_SparkHits_01
graph {
 properties {
  bEnableResponsiveAA=true
  bUsedWithNiagaraSprites=true
  bUsedWithParticleSprites=true
  blendMode=BLEND_Translucent
  pixelDepthOffsetMode=PDOM_Legacy
  shadingModel=MSM_Unlit
  usedShadingModels=Unlit
 }
 nodes [
  ConstantBiasScale(None=Sine(None=Time()))
  n1 = TextureSample() {samplerType=SAMPLERTYPE_LinearColor, texture=/Game/Effects/Fort_Effects/Textures/Sparks/T_SparkHits_2x2_01}
  n2 = ScalarParameter() {defaultValue=4, parameterName=Number of Steps}
  n3 = ParticleColor()
  n4 = Subtract(A=@n2, B=OneMinus(None=Multiply(A=@n2, B=@n3.A)))
  Divide(A=Floor(None=Multiply(A=@n1.G, B=@n4)), B=ScalarParameter() {defaultValue=4, parameterName=Number of Steps})
  MaterialFunctionCall(Emissive=Multiply(A=ParticleColor().RGB, B=@n1.G)) {materialFunction=&MF_FX_EmissiveAdjustments}
 ]
}
