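# Additive unlit spark material for Niagara and Cascade sprite particles, with responsive AA enabled.
# The red channel of a spark texture is boosted by a constant, tinted by particle color and scaled by a per-particle random brightness.
# That result is multiplied by the tinted mask again, squaring mask and color for higher contrast.
# Particle alpha scales the result before it feeds the MF_FX_EmissiveAdjustments function.
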Material /Game/Effects/Fort_Effects/Materials/Sparks/M_Spark_Cine_01
graph {
 properties {
  bEnableResponsiveAA=true
  bUsedWithNiagaraSprites=true
  bUsedWithParticleSprites=true
  blendMode=BLEND_Additive
  pixelDepthOffsetMode=PDOM_Legacy
  shadingModel=MSM_Unlit
  usedShadingModels=Unlit
 }
 nodes [
  n1 = Constant() {r=1.6}
  n2 = TextureSample() {texture=/Game/Weapons/FORT_Rifles/Effects/Textures/spark}
  n3 = ParticleColor()
  n4 = Multiply(A=LinearInterpolate(Alpha=ParticleRandom()) {constA=0.1}, B=Multiply(A=Multiply(A=@n1, B=@n2.R), B=@n3.RGB))
  n5 = Multiply(A=@n2.R, B=@n3.RGB)
  MaterialFunctionCall(Emissive=Multiply(A=Multiply(A=@n4, B=@n5), B=@n3.A)) {materialFunction=&MF_FX_EmissiveAdjustments}
 ]
}
