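# Unlit spark material for Cascade and Niagara mesh particles, giving a jittering, glowing orange-to-yellow look.
# A fast-panning fractal texture's red channel scales the vertex normal and blends between inner and outer offsets for jagged vertex displacement.
# The same noise blends low and high colors, boosted by an emissive multiplier.
# DeriveHDRfromLDR remaps that color, except on mobile and low quality, before MF_FX_EmissiveAdjustments.
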Material /Game/Effects/Fort_Effects/Materials/Sparks/M_Spark_JitterMesh
graph {
 properties {
  bUsedWithMeshParticles=true
  bUsedWithNiagaraMeshParticles=true
  pixelDepthOffsetMode=PDOM_Legacy
  shadingModel=MSM_Unlit
  usedShadingModels=Unlit
 }
 nodes [
  n1 = TextureSample(UVs=Panner(Coordinate=TextureCoordinate() {uTiling=0.5, vTiling=0.5}) {speedX=6, speedY=5}) {texture=/Game/Global/Textures/Default/Caustics/T_Fractal_01}
  n2 = Multiply(A=@n1.R, B=VertexNormalWS())
  LinearInterpolate(
   A=Multiply(A=@n2, B=ScalarParameter() {defaultValue=-10, parameterName=VertOffsetInner}) {constB=-10}
   B=Multiply(A=@n2, B=ScalarParameter() {defaultValue=500, parameterName=VertOffsetOuter}) {constB=800}
   Alpha=@n1.R
  ) {constA=-10, constB=400}
  n3 = LinearInterpolate(
   A=VectorParameter() {parameterName=ColorLow, defaultValue=<rgba 0.2969 0.0734 0 1>}.RGB
   B=VectorParameter() {parameterName=ColorHigh, defaultValue=<rgba 1 0.6527 0 1>}.RGB
   Alpha=@n1.R
  )
  n4 = Multiply(A=@n3, B=ScalarParameter() {defaultValue=200, parameterName=EmissiveMultiplier}) {constB=200}
  n5 = MaterialFunctionCall(LDR Input=@n3, HDR Tint=@n4) {materialFunction=&DeriveHDRfromLDR}
  MaterialFunctionCall(Emissive=QualitySwitch(Default=ShadingPathSwitch(Default=@n5.RemappedRange, Mobile=@n4) {desc=Migrated from FeatureLevelSwitch}, Low=@n4)) {materialFunction=&MF_FX_EmissiveAdjustments}
 ]
}
